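# Post-process AnimBP for the EssayViewMyth head that counter-rotates the head bone, apparently for its backwards-hat cosmetic.
# Each update scales the negated head-to-neck_02 ref-pose delta by per-axis percentages; only yaw's is nonzero here.
# The AnimGraph adds that rotation through two chained head Modify Bone nodes; a lower-teeth offset node is unconnected.
# A customization meta tag detects the hat, raising yaw correction while crouching or skydive-floating.
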AnimBlueprint /BRCosmetics/Characters/Player/Male/Medium/Heads/M_MED_EssayViewMyth_Head/Meshes/EssayViewMyth_Head_PostAnimBP
target_skeleton=/Game/Characters/Player/Common/Fortnite_Base_Head/Fortnite_Base_3L_Head_Skeleton
graphs [
 AnimGraph {nodes [
   "Transform (Modify) Bone - Bone: FACIAL_C_TeethLower"!(ComponentPose, bAlphaBoolEnabled:bool=True, Translation="0.1,0,-0.2", Rotation="0,0,0", Scale="1,1,1") -> (Pose) {
    alphaInputType=Bool
    rotationMode=BMM_Additive
    rotationSpace=BCS_BoneSpace
    scaleMode=BMM_Additive
    scaleSpace=BCS_BoneSpace
    translationMode=BMM_Additive
    translationSpace=BCS_BoneSpace
    alphaBoolBlend {blendInTime=1, blendOption=Linear, blendOutTime=1}
    boneToModify {boneName=FACIAL_C_TeethLower}
    translation {x=0.1, z=-0.2}
   }
   n1 = Get head rot() -> (head rot:Rotator)
   n2 = "Transform (Modify) Bone - Bone: head"!(ComponentPose=Local To Component(LocalPose=Input Pose()), bAlphaBoolEnabled:bool=True, Translation="0,0,0", Rotation=@n1, Scale="1,1,1") -> (Pose) {
    alphaInputType=Bool
    rotationMode=BMM_Additive
    rotationSpace=BCS_BoneSpace
    scaleMode=BMM_Additive
    scaleSpace=BCS_BoneSpace
    translationMode=BMM_Additive
    translationSpace=BCS_BoneSpace
    alphaBoolBlend {blendInTime=1, blendOption=Linear, blendOutTime=1}
    boneToModify {boneName=head}
    rotation=<rot 50 50 50>
   }
   n3 = "Transform (Modify) Bone - Bone: head"(ComponentPose=@n2, bAlphaBoolEnabled=True, Translation="0,0,0", Rotation=@n1, Scale="1,1,1") {
    alphaInputType=Bool
    rotationMode=BMM_Additive
    rotationSpace=BCS_BoneSpace
    scaleMode=BMM_Additive
    scaleSpace=BCS_BoneSpace
    translationMode=BMM_Additive
    translationSpace=BCS_BoneSpace
    alphaBoolBlend {blendInTime=1, blendOption=Linear, blendOutTime=1}
    boneToModify {boneName=head}
   }
   Output Pose(Result=Component To Local(ComponentPose=@n3))
  ]}
 EventGraph {nodes [
   Event BlueprintUpdateAnimation()
   n1 = GetOwnerSkeletalMeshComponent()
   n2 = BreakRotator(InRot=BreakTransform(InTransform=GetDeltaTransformFromRefPose(self=@n1, BoneName=head, BaseName=neck_02)).Rotation)
   n3 = Get Y percentage() -> (Y percentage:float)
   n4 = Get X percentage() -> (X percentage:float)
   n5 = Get Z percentage() -> (Z percentage:float)
   > Set head rot(head rot:Rotator=MakeRotator(
     Roll=float * float!(A=@n2, B=-1, C:double=@n4) -> (ReturnValue)
     Pitch=float * float(A=@n2.Pitch, B=-1, C=@n3)
     Yaw=float * float(A=@n2.Yaw, B=-1, C=@n5)
    )) -> (Output_Get:Rotator)
   n6 = Get bIsSkydiveFloating()
   n7 = Get bIsCrouching()
   > n8 = Branch(Condition=AND Boolean(A=Get BackwardsHat() -> (BackwardsHat:bool), B=OR Boolean(A=@n7, B=@n6)))
   > Set Apply head rotation(Apply head rotation:bool=true) -> (Output_Get:bool)
   > Set Z percentage(Z percentage:float=0.6) -> (Output_Get:float)
   Set Apply head rotation(execute=@n8.else, Apply head rotation=false)
   > Set Apply head rotation Forward(Apply head rotation Forward:bool=Get bIsCrouching()) -> (Output_Get:bool)
   > Set Z percentage(Z percentage=0.2)
   Event OnCharacterCustomizationLoaded()
   > Set BackwardsHat(BackwardsHat:bool=CharacterItemDefinitionHasAnyMetaTag(Tags=Make Array("[0]"=(TagName="Cosmetics.MetaTags.EssayViewMythHat")))) -> (Output_Get:bool)
   TryGetPawnOwner()
   OR Boolean(A=Get bIsPlayingEmote(), B=Get bIsInFrontEnd())
   Get bIsSkydiveFloating()
   Get BackwardsHat()
  ]}
]
variables [
 Apply head rotation {type=boolean}
 Apply head rotation Forward {type=boolean}
 BackwardsHat {type=boolean}
 head rot {type=Rotator}
 X percentage {type=number}
 Y percentage {type=number}
 Z percentage {type=number, default=0.6}
]
